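# Material function that tints static mesh materials with the terrain's biome color, sampled from a combined landscape colormap.
# MF_Athena_Terrain_Coords turns world position into UVs for the colormap, and the sampled RGB is brightened by a constant.
# MatLayerBlend_MultiplyBaseColor multiplies that color into the incoming material attributes' base color.
# Outputs are the blended material and the raw scaled color; the ColormapDivisor collection parameter is present but unconnected.
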MaterialFunction /FigmentEnvironment/LM_8_1_1/Athena/Environments/Materials/MF_StaticMeshBiomeColor
graph {nodes [
  CollectionParameter() {collection=/FigmentEnvironment/LM_8_1_1/Athena/Environments/Landscape/MPC/MPC_Landscape, parameterName=ColormapDivisor}
  n1 = TextureSample(UVs=MaterialFunctionCall!(WorldPosition) {materialFunction=/FigmentEnvironment/LM_8_1_1/Athena/Environments/Landscape/MaterialFunctions/Utilities/MF_Athena_Terrain_Coords}.Result) {
   samplerSource=SSM_Wrap_WorldGroupSettings
   samplerType=SAMPLERTYPE_LinearColor
   texture=/FigmentEnvironment/FN_900/Athena/Environments/Landscape/Textures/T_Figment_Terrain_CombinedColors_S9
  }
  n2 = Multiply(A=@n1.RGB, B=Constant() {r=3.3})
  FunctionOutput(None=MaterialFunctionCall!(
    Material=FunctionInput() {bUsePreviewValueAsDefault=true, inputName=Material, inputType=FunctionInput_MaterialAttributes}
    NewBaseColor=@n2
    Mask
   ) {materialFunction=/Engine/Functions/MaterialLayerFunctions/MatLayerBlend_MultiplyBaseColor}.Blended Material) {outputName=Material}
  FunctionOutput(None=@n2) {outputName=Color}
 ]}
